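# Sound cue for the looping in-air sound of the Stamina capsule environmental gameplay object.
# A looping wave feeds an attenuation node using a dedicated in-air settings asset, then an enveloper whose volume curve fades in from near silence.
# The enveloper enters a root mixer with an empty second input.
# The cue uses the gadgets and items sound class, caps concurrency at two, and carries long-range capsule attenuation overrides.
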SoundCue /StaminaGameplay/Gameplay/Environmental/Capsule/Audio/Stamina_Capsule_InAir_Loop_Cue
root=Mixer_0
nodes [
 {
  id=Attenuation_0
  children [WavePlayer_2]
  properties {attenuationSettings=/StaminaGameplay/Gameplay/Environmental/Capsule/Audio/Settings/Stamina_Capsule_InAir_Att}
 }
 {
  id=Enveloper_0
  children [Attenuation_0]
  properties {pitchCurve {editorCurveData {keys [{time=0}]}}, volumeCurve {editorCurveData {keys [<at 0 0.01>,<at 0.5 0.1>,<at 1 1>]}}}
 }
 {id=Mixer_0, children [Enveloper_0,null], properties {inputVolume [0.8,0.75]}}
 {
  id=WavePlayer_2
  properties {bLooping=true, soundWaveAssetPtr=&/Stamina_Capsule_InAir_Loop}
 }
]
properties {
 bOverrideConcurrency=true
 maxDistance=145000
 soundClassObject=&Fort_Gadgets_Items
 volumeMultiplier=0.9
 attenuationOverrides {attenuationShape=Capsule, distanceAlgorithm=Logarithmic, falloffDistance=125000, attenuationShapeExtents=<1000000 25000>}
 concurrencyOverrides {maxCount=2}
}
