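# Sound cue for the landing of the Stamina capsule, an environmental gameplay object.
# A Random node picks one of three equally weighted landing waves and feeds it through a Modulator at the root.
# Playback uses a dedicated capsule-land attenuation asset with a large max distance, routed to an important in-game mission sound class at reduced volume.
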SoundCue /StaminaGameplay/Gameplay/Environmental/Capsule/Audio/Stamina_Capsule_Land_Cue
root=Modulator_0
nodes [
 {id=Modulator_0, children [Random_0]}
 {id=Random_0, children [WavePlayer_1,WavePlayer_2,WavePlayer_3], properties {weights [1,1,1]}}
 {id=WavePlayer_1, properties {soundWaveAssetPtr=&/Stamina_Capsule_Land_03}}
 {id=WavePlayer_2, properties {soundWaveAssetPtr=&/Stamina_Capsule_Land_01}}
 {id=WavePlayer_3, properties {soundWaveAssetPtr=&/Stamina_Capsule_Land_02}}
]
properties {
 attenuationSettings=/StaminaGameplay/Gameplay/Environmental/Capsule/Audio/Settings/Stamina_Capsule_Land_Att
 maxDistance=80000
 soundClassObject=&Fort_Mission_Ingame_Important_PSM
 volumeMultiplier=0.65
}
