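# Sound cue for the opening of a capsule among the Stamina environmental gameplay objects.
# A single wave player feeds the capsule-open sound wave straight to the output at reduced volume, with no randomization or modulation.
# The cue is routed to the Fort_Explosives sound class and uses inline attenuation and concurrency overrides, with concurrency set to prevent new instances once the limit is reached.
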SoundCue /StaminaGameplay/Gameplay/Environmental/Capsule/Audio/Stamina_Capsule_Open_Cue
root=WavePlayer_0
nodes [{id=WavePlayer_0, properties {soundWaveAssetPtr=&/Stamina_Capsule_Open}}]
properties {
 bOverrideAttenuation=true
 bOverrideConcurrency=true
 maxDistance=2512
 soundClassObject=/Game/Sounds/FortSoundClassesAndModes/Fort_Explosives
 volumeMultiplier=0.3
 attenuationOverrides {falloffDistance=2000, attenuationShapeExtents {x=512}}
 concurrencyOverrides {resolutionRule=PreventNew}
}
