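# Sound cue for the spawn sound of a Stamina capsule, an environmental gameplay object.
# It picks one of three spawn wave variants with equal weight through a Random node, then a Modulator varies pitch and volume.
# It is routed to the gadgets and items sound class, references a dedicated attenuation asset for spatialization, and overrides concurrency to cap simultaneous instances.
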SoundCue /StaminaGameplay/Gameplay/Environmental/Capsule/Audio/Stamina_Capsule_Spawn_Cue
root=Modulator_0
nodes [
 {id=Modulator_0, children [Random_0], properties {pitchMax=1.1, pitchMin=0.85, volumeMax=0.9, volumeMin=0.8}}
 {id=Random_0, children [WavePlayer_1,WavePlayer_2,WavePlayer_3], properties {weights [1,1,1]}}
 {id=WavePlayer_1, properties {soundWaveAssetPtr=&/Stamina_Capsule_Spawn_03}}
 {id=WavePlayer_2, properties {soundWaveAssetPtr=&/Stamina_Capsule_Spawn_01}}
 {id=WavePlayer_3, properties {soundWaveAssetPtr=&/Stamina_Capsule_Spawn_02}}
]
properties {
 attenuationSettings=/StaminaGameplay/Gameplay/Environmental/Capsule/Audio/Settings/Stamina_Capsule_Spawn_Att
 bOverrideConcurrency=true
 maxDistance=140000
 soundClassObject=&Fort_Gadgets_Items
 volumeMultiplier=0.55
 attenuationOverrides {
  attenuationShape=Capsule
  distanceAlgorithm=Logarithmic
  falloffDistance=1250000
  lPFRadiusMax=3000
  lPFRadiusMin=2000
  attenuationShapeExtents=<10000000 250000>
 }
 concurrencyOverrides {maxCount=3}
}
