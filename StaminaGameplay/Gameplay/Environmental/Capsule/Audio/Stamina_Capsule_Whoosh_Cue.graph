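# Sound cue for the whoosh of the Stamina capsule, an environmental gameplay object, classed under gadget and item sounds.
# A short fixed delay feeds a random node that picks one of three equally weighted whoosh variations.
# It references the capsule landing attenuation asset and also carries long-range override values with distance low-pass filtering and listener-focus attenuation.
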SoundCue /StaminaGameplay/Gameplay/Environmental/Capsule/Audio/Stamina_Capsule_Whoosh_Cue
root=Delay_0
nodes [
 {id=Delay_0, children [Random_1], properties {delayMax=0.2, delayMin=0.2}}
 {id=Random_1, children [WavePlayer_0,WavePlayer_1,WavePlayer_2], properties {weights [1,1,1]}}
 {id=WavePlayer_0, properties {soundWaveAssetPtr=&/Stamina_Capsule_Whoosh_03}}
 {id=WavePlayer_1, properties {soundWaveAssetPtr=&/Stamina_Capsule_Whoosh_01}}
 {id=WavePlayer_2, properties {soundWaveAssetPtr=&/Stamina_Capsule_Whoosh_02}}
]
properties {
 attenuationSettings=/StaminaGameplay/Gameplay/Environmental/Capsule/Audio/Settings/Stamina_Capsule_Land_Att
 maxDistance=80000
 soundClassObject=&Fort_Gadgets_Items
 volumeMultiplier=0.7
 attenuationOverrides {
  bAttenuateWithLPF=true
  bEnableListenerFocus=true
  falloffDistance=15000
  lPFFrequencyAtMax=1000
  lPFRadiusMax=15000
  lPFRadiusMin=10000
  nonFocusVolumeAttenuation=0.35
  attenuationShapeExtents {x=5000}
 }
}
