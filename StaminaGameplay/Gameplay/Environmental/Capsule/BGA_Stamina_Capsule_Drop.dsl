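; Falling stamina capsule drop actor that plays flight and landing effects and orients the container it spawns.
; Cosmetic-capable instances play a spawn cue and in-air audio, then fade it on landing with a sound and indicator.
; Otherwise, the server polls squared distance to the landing position each second, tripling velocity and calling a whoosh once within range.
; The spawned container receives the drop's rotation, pitched 90 degrees, with a forced net update.
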(event EventBeginPlay
  (|CallParentFunction)
  (Networking|CanExecuteCosmeticEvents
    (:True
      (GameplayCue|ExecuteGameplayCue%28Burst%29Local self (Variables|Default|GetSpawnCue) (Ability|GameplayCue|MakeGameplayCueParameters 0.0))
      (Utilities|FlowControl|Delay 0.2)
      (bind _inairaudio (Variables|Default|GetInAirAudio)
        (:then
          (Audio|Components|Audio|Play _inairaudio 0.0))
        (:else)))
    (:False
      (CallFunction|CheckforSpeedup))))

(event EventOnLandedClient
  (bind _returnvalue (Transformation|GetActorLocation))
  (bind _landsound (Variables|Sound|GetLandSound))
  (bind _inairaudio (Variables|Default|GetInAirAudio)
    (:then
      (Audio|Components|Audio|FadeOut _inairaudio 0.5 0.0 "Linear" :!class /Script/Engine.AudioComponent))
    (:else
      _))
  (Audio|PlaySoundatLocation _landsound _returnvalue)
  (SoundIndicator|ShowSoundIndicatorforDuration (Variables|Default|GetFortSoundIndicator) _returnvalue _landsound 1.0 (Variables|/Script/Engine.SoundBase|GetDuration _landsound) 0))

(event Custom|CheckForSpeedup
  (if (Networking|IsServer)
    (if (< (Math|Vector|DistanceSquared(Vector) (Transformation|GetActorLocation) (Variables|Default|GetLandingPosition)) (* (Variables|Default|GetSpeedupDistance) (Variables|Default|GetSpeedupDistance)))
      (CallFunction|SetMovementVelocity (* (Variables|/Script/Engine.MovementComponent|GetVelocity (Variables|Default|GetFortProjectileMovement)) 3.0))
      (CallFunction|PlayWoosh)
      (else
        (Utilities|Time|SetTimerbyEvent (delegate Custom|Retrigger) 1.0)))))

(event Custom|Retrigger
  (CallFunction|CheckforSpeedup))

(event Custom|PlayWoosh
  (bind _inairaudio (Variables|Default|GetInAirAudio)
    (:then
      (Audio|SpawnSoundAttached (Variables|Sound|GetWhooshSound) _inairaudio))
    (:else)))

(fn OnSpawnedContainer
  (params (Container :type Actor%20Object%20Reference))
  (bind _ascontainer_stamina_capsule (Utilities|Casting|CastToContainer_Stamina_Capsule Container))
  (Variables|/StaminaGameplay/Gameplay/Environmental/Capsule/Container_Stamina_Capsule.Container_Stamina_Capsule_C|SetSpawnRot (Math|Rotator|MakeRotator (.roll (Transformation|GetActorRotation)) (- (.pitch (Transformation|GetActorRotation)) -9e+01) (.yaw (Transformation|GetActorRotation))) _ascontainer_stamina_capsule)
  (Networking|ForceNetUpdate _ascontainer_stamina_capsule)
  (return false))

(fn ConstructionScript
  (|CallParentFunction))
