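; Stamina capsule container that animates upright on spawn and can be opened by players.
; On BeginPlay the server logs spawn analytics to the game state's StaminaAnalytics component, while non-server instances apply a spawn rotation and fade in ambience.
; A self-retriggering short timer, used instead of Tick, eases the mesh rotation toward zero until nearly vertical.
; Opening fires a gameplay cue, fades out ambience and logs analytics with timing and the opening controller.
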(event EventBeginPlay
  ;[1 1 1] BeginPlay
  (|CallParentFunction)
  (Variables|Default|SetTickStartTime (Utilities|Time|GetTimeSeconds))
  (if (Networking|IsServer)
    (Class|StaminaAnalytics|AddStaminaEventAnalytics (Actor|GetComponentbyClass (Fortnite|Athena|GameState|GetGameStateAthena) "/Script/StaminaGameplayRuntime.StaminaAnalytics") self (Utilities|Struct|MakeStaminaCapsuleAnalytics (Variables|Default|GetTickStartTime) (Transformation|GetActorLocation self)) 0)
    (else
      (Transformation|SetRelativeRotation (Variables|Default|GetStaticMesh) (Variables|Default|GetSpawnRot) false false)
      (CallFunction|TickRotation)
      (Audio|Components|Audio|FadeIn (Variables|Default|GetAmbience) 1.5 1.0 0.0 "Linear" :!class /Script/Engine.AudioComponent))))

(event Custom|TickRotation
  ;[1 1 1] TickRotation
  (bind _returnvalue (Fortnite|GetAngleDegrees (Math|Vector|GetUpVector (Variables|/Script/Engine.SceneComponent|GetRelativeRotation (Variables|Default|GetStaticMesh))) (Math|Vector|MakeVector 0.0 0.0 1.0)))
  (if (> _returnvalue 0.1)
    (Transformation|SetRelativeRotation (Variables|Default|GetStaticMesh) (Math|Rotator|Lerp(Rotator) (Variables|/Script/Engine.SceneComponent|GetRelativeRotation (Variables|Default|GetStaticMesh)) "0, 0, 0" (Utilities|RemapFloat (- (Utilities|Time|GetTimeSeconds) (Variables|Default|GetTickStartTime)) "(X=0.000000,Y=1.200000)" "(X=0.000000,Y=1.000000)" true) true) false false)
    (Utilities|Time|SetTimerbyEvent (delegate Custom|Retrigger) 0.016)))

(event Custom|Retrigger
  ;[1 1 1] TickRotation
  (CallFunction|TickRotation))

(event EventOpenContainer
  (params (Target :type Pawn%20Object%20Reference))
  ;[1 0.9743 0.3585] Open Container Event
  (|CallParentFunction Target)
  (GameplayCue|ExecuteGameplayCue self (Variables|Default|GetOpenCue) (Utilities|Struct|MakeGameplayEffectContextHandle))
  (CallFunction|StopAmbientSound)
  (Class|StaminaAnalytics|AddStaminaEventAnalytics (Actor|GetComponentbyClass (Fortnite|Athena|GameState|GetGameStateAthena) "/Script/StaminaGameplayRuntime.StaminaAnalytics") self (Utilities|Struct|MakeStaminaCapsuleAnalytics (Variables|Default|GetTickStartTime) (Transformation|GetActorLocation self) (Utilities|Time|GetTimeSeconds)) (Fortnite|GetFortPlayerControllerfromActor Target)))

(event Custom|StopAmbientSound
  (bind _ambience (Variables|Default|GetAmbience)
    (:then
      (Audio|Components|Audio|FadeOut _ambience 0.5 0.0 "Linear" :!class /Script/Engine.AudioComponent))
    (:else)))

(fn ConstructionScript
  (|CallParentFunction))
